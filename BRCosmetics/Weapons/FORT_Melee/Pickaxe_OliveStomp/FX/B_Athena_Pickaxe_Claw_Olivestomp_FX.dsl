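; Weapon FX Blueprint for the OliveStomp dual-wield claw pickaxe, driving its swing, impact, idle and deploy effects.
; Swings and impacts pass montage section indices to effects; frontend preview spawns its own counter-cycled directional impact.
; Impacts are oriented from camera vectors, with weakpoint hits using a camera trace and replaying locally, then through server-to-client events.
; Deploy FX spawns on visibility or matching emotes, and a cosmetic variant check sets an idle FX parameter.
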(event EventSwingRight_Common
  ;[0 0 0] swingsectionindex
  (|CallParentFunction)
  (bind _section_index (Utilities|SetupSwingMontageSectionIndex (Variables|Actor|GetInstigator))
    (:Out
      (Effects|Components|ParticleSystem|SetIntParameter (CallFunction|GetSwingVFXComponent) "SwingSectionIndex" _section_index)
      (Components|Activation|Deactivate (Variables|Default|GetDeployFX)))))

(event EventSwingLeft_Common
  ;[0 0 0] swingsectionindex
  (|CallParentFunction)
  (bind _section_index (Utilities|SetupSwingMontageSectionIndex (Variables|Actor|GetInstigator))
    (:Out
      (Effects|Components|ParticleSystem|SetIntParameter (CallFunction|GetSwingOffhandVFXComponent) "SwingSectionIndex" _section_index)
      (Components|Activation|Deactivate (Variables|Default|GetDeployFX)))))

(event FortWeapon|EventOnPlayImpactFX
  (params (HitResult :type Hit%20Result%20Structure%20%28by%20ref%29) (ImpactPhysicalSurface :type EPhysicalSurface%20Enum :default "SurfaceType_Default") (SpawnedPSC :type FXSystem%20Component%20Object%20Reference))
  (bind (_returnvalue_location _returnvalue_rotation _returnvalue_scale) (Math|Transform|BreakTransform (Transformation|GetActorTransform (Variables|Default|GetAs%20Fort%20Player%20Pawn))))
  (bind (_returnvalue_location_2 _returnvalue_rotation_2) (Math|Transform|BreakTransform (Transformation|GetActorTransform (Game|GetPlayerCameraManager 0))))
  (|CallParentFunction HitResult ImpactPhysicalSurface SpawnedPSC)
  (Variables|Default|SetImpact%20FX SpawnedPSC)
  (if (Reactive|Frontend|IsCosmeticObjectDisplayedasPreview 0)
    ;[0 0 0] frontend
    (Components|Activation|Deactivate SpawnedPSC)
    (bind _returnvalue (Niagara|SpawnSystematLocation "/BRCosmetics/Weapons/FORT_Melee/Pickaxe_OliveStomp/FX/NS_Pickaxe_OliveStomp_DirectionalImpact.NS_Pickaxe_OliveStomp_DirectionalImpact" (Math|Transform|TransformLocation (Transformation|GetActorTransform (Variables|Default|GetAs%20Fort%20Player%20Pawn)) "100.000000,0.000000,50.000000") (Math|Rotator|SelectRotator _returnvalue_rotation_2 (Math|Rotator|CombineRotators _returnvalue_rotation "0.000000,180.000000,0.000000") (Reactive|Frontend|IsCosmeticObjectDisplayedasPreview 0)) _returnvalue_scale true false))
    (Utilities|FlowControl|DoOnce false
      (:Completed
        (Effects|Components|ParticleSystem|SetIntParameter _returnvalue "SwingSectionIndex" (Math|Integer|%25%28Integer%29 (Variables|Default|GetFE_Counter) 4))
        (Effects|Components|ParticleSystem|SetBoolParameter _returnvalue "bIsFrontEndPreview" true)
        (Components|Activation|Activate _returnvalue false)
        (Math|Integer|IncrementInt (Variables|Default|GetFE_Counter)
          (:%20%20
            (Utilities|FlowControl|RetriggerableDelay 0.25) ; prevents duplicated signal
          ))))
    (else
      (bind _section_index (Utilities|SetupSwingMontageSectionIndex (Variables|Actor|GetInstigator))
        (:Out
          (Effects|Components|ParticleSystem|SetIntParameter (Variables|Default|GetImpact%20FX) "SwingSectionIndex" _section_index)
          (switch Utilities|FlowControl|Switch|SwitchonEPhysicalSurface ImpactPhysicalSurface
            (:SurfaceType4
              _)
            (:SurfaceType5
              _)
            (:SurfaceType12
              (CallFunction|WeakpointEvent)
              (CallFunction|CEFXServer))
            (:SurfaceType14
              _)
            (:SurfaceType15
              _)
            (:SurfaceType16
              _))
          (Utilities|FlowControl|Delay 0.05)
          (CallFunction|SetupHumanHit))))))

(event Custom|WeakpointEvent
  ;[0 0 0] weakspot event- play event on client, THEN play it on every client. Allows for non-delayed impact on owner client, while keeping the visuals replicated
  (CallFunction|SetupWeakspotEvent))

(event Custom|CE_FX_Server
  ;[0 0 0] weakspot event- play event on client, THEN play it on every client. Allows for non-delayed impact on owner client, while keeping the visuals replicated
  (CallFunction|CEFXClient))

(event Custom|CE_FX_Client
  ;[0 0 0] weakspot event- play event on client, THEN play it on every client. Allows for non-delayed impact on owner client, while keeping the visuals replicated
  (CallFunction|SetupWeakspotEvent))

(event FortWeapon|EventOnUnEquippedAndDetached
  ;[0 0 0] kill system on unequip
  (|CallParentFunction)
  (Components|Activation|Deactivate (Variables|Default|GetDeployFX)))

(event EventOnWeaponVisibilityChanged
  (params (bVisible :type Boolean) (bSetForLocalControllerOnly :type Boolean))
  ;[0 0 0] Deploy FX(Use one System for both hand, allowing to only show the FX on one of the hands!)
  (|CallParentFunction bVisible bSetForLocalControllerOnly)
  (if bVisible
    (Development|PrintString "PICKAXE VISIBLE")
    (Effects|Components|ParticleSystem|SetIntParameter (CallFunction|GetIdleOffhandVFXComponent) "DualWieldHand" 1)
    (Utilities|FlowControl|Sequence
      (:then_0
        (bind _asfort_player_pawn (Utilities|Casting|CastToFortPlayerPawn (Variables|Actor|GetInstigator)))
        (bind _output_get (Variables|Default|SetAs%20Fort%20Player%20Pawn _asfort_player_pawn))
        (Utilities|FlowControl|Delay 0.05)
        (if (Math|Boolean|ANDBoolean (not (Variables|/Script/FortniteGame.FortPlayerPawn|GetbIsTacticalSprinting _asfort_player_pawn)) (not (Variables|Default|GetbAlreadyDeployed)) (not (Reactive|Frontend|IsCosmeticObjectDisplayedasPreview 0)) (not (Variables|/Script/FortniteGame.FortPlayerPawn|GetbIsSkydiving _output_get)))
          (bind _returnvalue (Niagara|SpawnSystemAttached "/BRCosmetics/Weapons/FORT_Melee/Pickaxe_OliveStomp/FX/NS_Pickaxe_OliveStomp_Deploy.NS_Pickaxe_OliveStomp_Deploy" (Variables|/Script/Engine.Character|GetMesh (Variables|Default|GetAs%20Fort%20Player%20Pawn)) "Root"))
          (Variables|Default|SetDeployFX _returnvalue)))
      (:then_1
        (Utilities|FlowControl|Delay 0.2)
        (CallFunction|VariantSetup) ; 💀Necessary because of emote swap notify / Loadout Augmentation
      ))))

(event FortWeapon|EventOnEquippedAndAttached
  ;[0 0 0 0.5] emote
  (|CallParentFunction)
  (Utilities|FlowControl|Sequence
    (:then_0
      (bind _asfort_player_pawn (Utilities|Casting|CastToFortPlayerPawn (Variables|Actor|GetInstigator)))
      (Utilities|FlowControl|Delay 0.04)
      (bind _output_get (Variables|Default|SetAs%20Fort%20Player%20Pawn _asfort_player_pawn))
      (bind _output_get_2 (Variables|Default|SetbEmoting (Math|Boolean|ORBoolean (== (Animation|GetCurrentMontage _output_get :!class /Script/Engine.Character) "/BRCosmetics/Animation/Game/MainPlayer/Emotes/Olivestomp/Emote_Olivestomp_CMF_M.Emote_Olivestomp_CMF_M") (== (Animation|GetCurrentMontage _output_get :!class /Script/Engine.Character) "/BRCosmetics/Animation/Game/MainPlayer/Emotes/Olivestomp/Emote_Olivestomp_CMM_M.Emote_Olivestomp_CMM_M") (== (Animation|GetCurrentMontage _output_get :!class /Script/Engine.Character) "/BRCosmetics/Animation/Game/MainPlayer/Emotes/Olivestomp/Emote_Olivestomp_CLM_M.Emote_Olivestomp_CLM_M") (== (Animation|GetCurrentMontage _output_get :!class /Script/Engine.Character) "/BRCosmetics/Animation/Game/MainPlayer/Emotes/Olivestomp/Emote_Olivestomp_CLF_M.Emote_Olivestomp_CLF_M") (== (Animation|GetCurrentMontage _output_get :!class /Script/Engine.Character) "/BRCosmetics/Animation/Game/MainPlayer/Emotes/Olivestomp/Emote_Olivestomp_CMM_Hero_M.Emote_Olivestomp_CMM_Hero_M"))))
      (if _output_get_2
        (if (not (Variables|Default|GetbAlreadyDeployed))
          (bind _returnvalue (Niagara|SpawnSystemAttached "/BRCosmetics/Weapons/FORT_Melee/Pickaxe_OliveStomp/FX/NS_Pickaxe_OliveStomp_Deploy.NS_Pickaxe_OliveStomp_Deploy" (Variables|/Script/Engine.Character|GetMesh (Variables|Default|GetAs%20Fort%20Player%20Pawn)) "Root"))
          (bind _output_get_3 (Variables|Default|SetDeployFX _returnvalue))
          (Niagara|SetNiagaraVariable(Float) _output_get_3 "isEmote" (Math|Conversions|ToFloat(Boolean) (Variables|Default|GetbEmoting))) ; make system aware it's being played as emote and do according tweaks
          (Development|PrintString "Spawned Emote Deploy FX")
          (Variables|Default|SetbAlreadyDeployed true)
          (EventDispatchers|AssignOnPawnStoppedEmote (Variables|Default|GetAs%20Fort%20Player%20Pawn) (delegate Custom|OnPawnStoppedEmote_Event))
          (if (or (== (Animation|Montage|MontageGetCurrentSection (Components|SkeletalMesh|GetAnimInstance (Variables|/Script/Engine.Character|GetMesh _output_get)) (Animation|GetCurrentMontage _output_get :!class /Script/Engine.Character)) "Retract") (and (Reactive|Frontend|IsCosmeticObjectDisplayedasPreview 0) (not (Variables|Default|GetbEmoting))))
            (Components|Activation|Deactivate _output_get_3)))))
    (:then_1
      (Utilities|FlowControl|Delay 0.2)
      (CallFunction|VariantSetup) ; 💀Necessary because of emote swap notify / Loadout Augmentation
    )))

(event Custom|OnPawnStoppedEmote_Event
  (params (MontageItemDef :type Fort%20Item%20Definition%20Object%20Reference) (PawnEmoting :type Fort%20Pawn%20Object%20Reference))
  ;[0 0 0 0.5] emote
  (Variables|Default|SetbAlreadyDeployed false)
  (Components|Activation|Deactivate (Variables|Default|GetDeployFX))
  (EventDispatchers|UnbindEventfromOnPawnStoppedEmote (Variables|Default|GetAs%20Fort%20Player%20Pawn) (delegate Custom|OnPawnStoppedEmote_Event))
  (Development|PrintString "Event UNBOUND!"))

(fn SetupHumanHit
  (bind (_returnvalue_x _returnvalue_y) (Math|Vector|BreakVector (* (- (Transformation|GetActorLocation) (Camera|GetCameraLocation (Game|GetPlayerCameraManager 0))) 4e+02)))
  (Transformation|SetWorldRotation (Variables|Default|GetImpact%20FX) (Math|Rotator|CombineRotators (Math|Conversions|RotationFromXVector (Math|Vector|MakeVector _returnvalue_x _returnvalue_y)) "0.000000,180.000000,0.000000") false false) ; wee-bit cursed, but line trace was failing if the player was in a weird angle or really close ;/
)

(fn SetupWeakspotEvent
  (bind _outhit (Collision|LineTraceByChannel (Camera|GetCameraLocation (Game|GetPlayerCameraManager 0)) (+ (* (Math|Vector|GetForwardVector (Camera|GetCameraRotation (Game|GetPlayerCameraManager 0))) 4e+03) (Camera|GetCameraLocation (Game|GetPlayerCameraManager 0))) "TraceTypeQuery10"))
  (bind (_outhit_bblockinghit _outhit_binitialoverlap _outhit_time _outhit_distance _outhit_location _outhit_impactpoint _outhit_normal) (Collision|BreakHitResult _outhit))
  (bind _output_get (Variables|Default|SetImpactRotation _outhit_normal))
  (bind (_returnvalue_x _returnvalue_y) (Math|Vector|BreakVector (* _output_get 4e+02)))
  (Transformation|SetWorldRotation (Variables|Default|GetImpact%20FX) (Math|Conversions|RotationFromXVector (Math|Vector|MakeVector _returnvalue_x _returnvalue_y)) false false)
  (Effects|Components|ParticleSystem|SetBoolParameter (Variables|Default|GetImpact%20FX) "WeakpointHit" true)
  (Development|PrintString "Hello" true true "(R=1.000000,G=0.201102,B=0.496928,A=1.000000)"))

(fn ConstructionScript
  (|CallParentFunction))

(fn VariantSetup
  (bind _asfort_player_pawn (Utilities|Casting|CastToFortPlayerPawn (Variables|Actor|GetInstigator)))
  (bind (_index _bindexisvalid _bfoundchannel) (CosmeticLoadout|GetVariantChannelActiveIndexforSlotActor "(TagName=\"LoadoutSlot.Pickaxe\")" _asfort_player_pawn "(TagName=\"Cosmetics.Variant.Channel.ProfileLoadout\")" (Utilities|Array|MakeArray "(TagName=\"Cosmetics.Variant.Property.Stage0\")") false false))
  (Effects|Components|ParticleSystem|SetFloatParameter (CallFunction|GetIdleVFXComponent) "BuiltIn" (Math|Conversions|ToFloat(Boolean) _bfoundchannel))
  (Development|PrintString "CHNNALE SIWTCH!!"))
