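# Translucent unlit, two-sided material for letter mesh particles in the OliveStomp pickaxe effects.
# A vertex offset chain shears the mesh along local X by height in its bounds, driven by the Skew dynamic parameter.
# It adds a per-particle randomized sine wiggle on local Z, whose strength changes over particle lifetime, plus a camera offset.
# A tint multiplied by particle color feeds the shared emissive adjustment function.
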Material /BRCosmetics/Weapons/FORT_Melee/Pickaxe_OliveStomp/FX/M_OliveStomp_Letter
graph {
 properties {
  bAutomaticallySetUsageInEditor=false
  bCastRayTracedShadows=false
  bUsedWithNiagaraMeshParticles=true
  blendMode=BLEND_Translucent
  pixelDepthOffsetMode=PDOM_Legacy
  shadingModel=MSM_Unlit
  twoSided=true
  usedShadingModels=Unlit
 }
 nodes [
  n1 = Multiply(
   A=Transform(None=Constant3Vector() {constant=<rgba 1 0 0 0>}) {transformSourceType=TRANSFORMSOURCE_Local}
   B=DynamicParameter() {defaultValue=<rgba 0 1 1 1>, paramNames [Skew,Param2,Param3,Param4]}.Skew
  )
  n2 = ConstantBiasScale(None=MaterialFunctionCall() {materialFunction=&BoundingBoxBased_0-1_UVW}.B) {bias=-0.5, scale=2}
  n3 = Time()
  n4 = ParticleRandom()
  n5 = Sine(None=Multiply(A=Add(A=@n3, B=@n4), B=ScalarParameter() {defaultValue=3, parameterName=WiggleSpeed}) {constB=1.5})
  n6 = ScalarParameter() {defaultValue=2, parameterName=Wiggle Strength}
  n7 = LinearInterpolate(Alpha=ParticleRelativeTime()) {constA=10}
  n8 = MaterialFunctionCall(B=Multiply(A=ConstantBiasScale(None=@n5) {bias=1.5}, B=Multiply(A=@n6, B=@n7))) {materialFunction=&AppendMany}
  Add(
   A=Add(A=Multiply(A=@n1, B=@n2), B=Transform(None=@n8.RGB) {transformSourceType=TRANSFORMSOURCE_Local})
   B=MaterialFunctionCall(Offset Amount=ScalarParameter() {parameterName="Offset Amount (S)"}) {materialFunction=/Engine/Functions/Engine_MaterialFunctions02/WorldPositionOffset/CameraOffset}.Clamped Camera Offset
  )
  MaterialFunctionCall(Emissive=Multiply(A=VectorParameter() {parameterName=Color, defaultValue=<rgba 1 0 0.0268 1>}.RGB, B=ParticleColor().RGB)) {materialFunction=&MF_FX_EmissiveAdjustments}
  Sign(None=@n2)
  VertexColor()
 ]
}
