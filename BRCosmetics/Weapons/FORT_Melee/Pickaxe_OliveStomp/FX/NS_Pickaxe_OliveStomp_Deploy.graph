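# Niagara deploy effect for the OliveStomp pickaxe that pops up comic-style "Shik" mesh lettering beside the weapon.
# A single burst spawns two five-mesh sets on two skeletal sockets, zero-scaling one set by comparing camera distances to those sockets.
# Letters are camera-oriented in a staggered layout, scale in and out in sequence, and take a time-of-day tint.
# User.isEmote blends camera and position offsets; a second mesh renderer accompanies a computed outline size.
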NiagaraSystem /BRCosmetics/Weapons/FORT_Melee/Pickaxe_OliveStomp/FX/NS_Pickaxe_OliveStomp_Deploy
dependencies {
 usedDataInterfaces [
  &NiagaraDataInterfaceCamera
  &NiagaraDataInterfaceCurve
  &NiagaraDataInterfaceArrayFloat
  &NiagaraDataInterfaceVectorCurve
  &NiagaraDataInterfaceSkeletalMesh
  &NiagaraDataInterfaceColorCurve
 ]
 usedDynamicInputs [
  &RandomRangeFloat
  &FNTimeOfDayScalar
  &ColorFromCurve
 ]
 usedModules [
  &self:SystemSpawnScript.SystemScriptSource.SystemScriptGraph.NiagaraNodeAssignment_3.SetVariables_#1
  &SystemState
  &Timeline
  &self:SystemSpawnScript.SystemScriptSource.SystemScriptGraph.NiagaraNodeAssignment_0.SetVariables_#2
  &self:SystemSpawnScript.SystemScriptSource.SystemScriptGraph.NiagaraNodeAssignment_1.SetVariables_#3
  &self:SystemSpawnScript.SystemScriptSource.SystemScriptGraph.NiagaraNodeAssignment_2.SetVariables_#4
 ]
}
emitters {ComicLetters_Shik {
  emitterUpdateScript [
   &EmitterState()
   &SpawnBurst_Instantaneous(Spawn Count=10, Spawn Time=0.05, Use Spawn Probability=true, UseLoopCountLimit=true)
  ]
  particleSpawnScript [
   &PartitionParticles(Particles Per Partition=5, Partition Mode:ENiagara_PartitionMode=::NewEnumerator1)
   &SkeletalMeshLocation(
    Bone Sampling Mode=::NewEnumerator4
    MaskSpawnGroup=false
    Orientation Sampling=::NewEnumerator1
    Position Sampling=::NewEnumerator2
    Skeletal Mesh=$System.Skeletal Mesh
    Use Sampled Mesh Attribute as Ribbon ID=false
   ) {module=SkeletalMeshLocation001}
   &SkeletalMeshLocation(
    Bone / Socket Index=1
    Bone Sampling Mode=::NewEnumerator4
    MaskSpawnGroup=false
    Orientation Sampling=::NewEnumerator1
    Position Sampling=::NewEnumerator2
    Skeletal Mesh=$System.Skeletal Mesh
    Use Sampled Mesh Attribute as Ribbon ID=false
   ) {module=SkeletalMeshLocation002}
   &InitializeParticle(
    Color=&LerpLinearColors(StartColor=<rgba 1 1 1 1>, EndColor=$System.TODColor, LerpFactor=0.7)
    Lifetime=0.75
    Mesh Scale Mode=::NewEnumerator3
    Mesh Uniform Scale=&Multiply_Float(
     A=0.05
     B=&MakeCustomFloatFromBool(
      Bool=&SetBoolByFloatComparison(
       A=&DistanceBetweenPositions(
        A=&ReturnCameraPosition(Coordinate Space:ENiagaraCoordinateSpace=::Simulation, Camera Query:NiagaraDataInterfaceCamera=DataInterface())
        B=$Particles.Initial.SkeletalMeshLocation001.SampledPosition
       )
       B=&DistanceBetweenPositions(
        A=&ReturnCameraPosition(Coordinate Space:ENiagaraCoordinateSpace=::Simulation, Camera Query:NiagaraDataInterfaceCamera=DataInterface())
        B=$Particles.Initial.SkeletalMeshLocation002.SampledPosition
       )
       Comparison Type=::NewEnumerator6
      )
      True Float=&SelectFloatFromArray(Direct Array Index=$Particles.PartitionParticles.Partition, Float Selection Array=DataInterface() {propertyValues {FloatData [1,0]}})
      False Float=&SelectFloatFromArray(Direct Array Index=$Particles.PartitionParticles.Partition, Float Selection Array=DataInterface() {propertyValues {FloatData [0,1]}})
     )
    )
    Position Mode=::NewEnumerator0
    Write Material Random=true
   )
   &InitialMeshOrientation(Mesh Orientation Mode=::NewEnumerator7, Rotation=<0 0 -0.25>, Use Rotation Vector=true)
   /SetVariables(Particles.MeshIndex=$Particles.PartitionParticles.Index In Partition) {isSetParameters=true}
   &CameraOffset(Camera Offset Amount=&Lerp_Float(
     A=&Add_Float(A=20, B=&ReturnNormalizedExecIndex(Normalized Index Scale=60))
     B=&Add_Float(A=5, B=&ReturnNormalizedExecIndex(Normalized Index Scale=20))
     Alpha=$User.isEmote
    ))
  ]
  particleUpdateScript [
   &ParticleState(Kill Particles When Lifetime Has Elapsed=false)
   &SkeletalMeshLocation(
    Bone / Socket Index=$Particles.PartitionParticles.Partition
    Bone Sampling Mode=::NewEnumerator4
    MaskSpawnGroup=false
    Orientation Sampling=::NewEnumerator1
    Skeletal Mesh=$System.Skeletal Mesh
   )
   /SetVariables(Particles.Position=&AddVectorToPosition(
     Position=$Particles.SkeletalMeshLocation.SampledPosition
     Vector=&RotateVectorByQuaternion(
      Vector To Rotate=&MakeVector(
       X=&Multiply_FloatByInt(Float=0, Integer=$Particles.PartitionParticles.Index In Partition)
       Y=&Add_Float(
        A=&Lerp_Float(
         A=0
         B=&Lerp_Float(
          A=20
          B=-20
          Alpha=&MakeCustomFloatFromBool(
           Bool=&SetBoolByFloatComparison(
            A=&MakeFloatFromVector(
             Vector=/Niagara/DynamicInputs/Vectors/CrossProduct(
              Vector A:Vector3f=&ReturnCameraVector(Camera Vectors=::NewEnumerator1, Coordinate Space=::Local)
              Vector B:Vector3f=&RotateVectorByQuaternion(Vector To Rotate=<0 0 1>, Quaternion=$Particles.SkeletalMeshLocation001.SampledRotation)
              Normalize Incoming Vectors:NiagaraBool=true
             )
             Channel=::NewEnumerator1
            )
            B=0.3
           )
           True Float=1
           False Float=0
          )
         )
         Alpha=$User.isEmote
        )
        B=&Multiply_FloatByInt(Float=5, Integer=$Particles.PartitionParticles.Index In Partition)
       )
       Z=&Add_Float(
        A=4
        B=&Add_Float(
         A=&Add_Float(
          A=&Lerp_Float(A=30, B=40, Alpha=$User.isEmote)
          B=&FloatFromCurve(
           FloatCurve=DataInterface() {propertyValues {Curve {keys [
               {arriveTangent=2.904, interpMode=RCIM_Cubic, leaveTangent=2.904, tangentMode=RCTM_User, time=0, value=0}
               {interpMode=RCIM_Cubic, tangentMode=RCTM_User, time=1, value=1}
              ]}}}
           Scale Curve=-30
          )
         )
         B=&Multiply_FloatByInt(Float=-4, Integer=$Particles.PartitionParticles.Index In Partition)
        )
       )
      )
      Quaternion=&OrientMeshToVector(
       Look At Direction=&ReturnCameraVector(Camera Vectors=::NewEnumerator3, Coordinate Space=::World)
       Reference Axis Vector=<1 0 0>
      )
     )
    )) {isSetParameters=true}
   &ScaleMeshSize(Scale Factor=&VectorFromFloat(Value=&FloatFromCurve(
      FloatCurve=DataInterface() {propertyValues {Curve {keys [
          {interpMode=RCIM_Cubic, tangentMode=RCTM_User, time=0, value=0}
          {interpMode=RCIM_Cubic, tangentMode=RCTM_User, time=0.1, value=1.5}
          {interpMode=RCIM_Cubic, tangentMode=RCTM_User, time=0.2, value=1}
          {interpMode=RCIM_Cubic, tangentMode=RCTM_User, time=0.7, value=1}
          {interpMode=RCIM_Cubic, leaveTangent=5, time=0.9, value=0}
         ]}}}
      CurveIndex=&Subtract_Float(
       A=$Particles.NormalizedAge
       B=&Multiply_Float(A=$Particles.PartitionParticles.Normalized Index In Partition, B=0.2)
      )
     )))
   &ScaleMeshSize(Scale Factor=/Niagara/DynamicInputs/ScaleByCurve/ScaleVectorByCurve(
     InitialValue:Vector3f=<1 1 1>
     ScaleCurve:NiagaraDataInterfaceVectorCurve=DataInterface() {propertyValues {
       XCurve {keys [
         <at 0 0 RCIM_Cubic>
         {arriveTangent=2.2843, interpMode=RCIM_Cubic, leaveTangent=2.2843, time=0.3178, value=1.4}
         {arriveTangent=-0.9254, interpMode=RCIM_Cubic, leaveTangent=-0.9254, time=0.4378, value=1}
         <at 0.75 1 RCIM_Cubic>
         <at 1 1 RCIM_Cubic>
        ]}
       YCurve {keys [
         <at 0 0 RCIM_Cubic>
         {arriveTangent=2.2886, interpMode=RCIM_Cubic, leaveTangent=2.2886, time=0.3167, value=4}
         {arriveTangent=-6.9239, interpMode=RCIM_Cubic, leaveTangent=-6.9239, time=0.5159, value=1.1808}
         {arriveTangent=-0.3735, interpMode=RCIM_Cubic, leaveTangent=-0.3735, time=0.75, value=1}
         <at 1 1 RCIM_Cubic>
        ]}
       ZCurve {keys [
         <at 0 0 RCIM_Cubic>
         {arriveTangent=3.0427, interpMode=RCIM_Cubic, leaveTangent=3.0427, time=0.25, value=1.3258}
         {arriveTangent=-0.4, interpMode=RCIM_Cubic, leaveTangent=-0.4, time=0.37, value=1.1258}
         {arriveTangent=-0.1997, interpMode=RCIM_Cubic, leaveTangent=-0.1997, time=0.75, value=1.1258}
         <at 1 1 RCIM_Cubic>
        ]}
      }}
     CurveIndex:NiagaraFloat=$Particles.NormalizedAge
    )) {module=ScaleMeshSize002}
   /SetVariables(Particles.OutlineSize:Vector3f=&Multiply_Vector(A=$Particles.Scale, B=<1.1454 1 1.1433>)) {isSetParameters=true}
   &Drag(Drag=4, Use Linear Drag=true, Use Rotational Drag=true)
   &SolveForcesAndVelocity(Clamp Velocity=false, Limit Acceleration=false)
   &DynamicMaterialParameters(
    Index 0 Param 1=&FloatFromCurve(
     FloatCurve=DataInterface() {propertyValues {Curve {keys [{tangentMode=RCTM_User, time=0, value=0}, {tangentMode=RCTM_User, time=1, value=1}]}}}
     Scale Curve=40
    )
    Module.Index0Param0Write=true
    Module.Index0Param1Write=true
    Module.Index0Param2Write=true
    Module.Index0Param3Write=true
   )
  ]
  rendererClasses [&NiagaraMeshRendererProperties]
  renderers [
   {rendererClass=&NiagaraMeshRendererProperties, rendererIndex=0}
   {rendererClass=&NiagaraMeshRendererProperties, rendererIndex=1}
  ]
 }}
summary {
 systemName=NS_Pickaxe_OliveStomp_Deploy
 emitters [{emitterName=ComicLetters_Shik, rendererClasses [&NiagaraMeshRendererProperties]}]
 userVariables [
  User.bIsSwinging {type=NiagaraBool}
  User.Debug {type=NiagaraFloat, defaultValue=0, description=""}
  User.DualWieldHand {type=NiagaraInt32}
  User.isEmote {type=NiagaraFloat, defaultValue=0, description=""}
  User.SwingSectionIndex {type=NiagaraInt32}
 ]
}
